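# Sound cue for the DwMl cosmetic pet, played by its name at the start of the down-but-not-out (DBNO) state.
# A quality-level node sends its first two slots to an equally weighted random pick between two waves, while the last two play only the first wave.
# It uses close-range pet attenuation and the creature-pets sound class, with boosted volume.
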SoundCue /BRCosmetics/Sounds/Pets/DwMl/PT_DwMl_DbnoStart_Cue
root=QualityLevel_0
nodes [
 {id=QualityLevel_0, children [Random_1,Random_1,WavePlayer_3,WavePlayer_3]}
 {id=Random_1, children [WavePlayer_3,WavePlayer_4], properties {weights [1,1]}}
 {id=WavePlayer_3, properties {soundWaveAssetPtr=&/PT_DwMl_DbnoStart_01}}
 {id=WavePlayer_4, properties {soundWaveAssetPtr=&/PT_DwMl_DbnoStart_02}}
]
properties {
 attenuationSettings=&Fort_Pets_Close_Athena_Att
 maxDistance=1800.0001
 soundClassObject=&Fort_Creatures_Pets
 volumeMultiplier=1.68
}
