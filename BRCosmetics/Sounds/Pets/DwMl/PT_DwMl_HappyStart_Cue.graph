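# Sound cue for the DwMl pet's HappyStart sound, played in the pet creatures sound class with close-range pet attenuation.
# A quality level node sends its first two levels to an equal-weight random pick among four Treasure_17 wave variants.
# The last two levels play only the base Treasure_17 wave, dropping the variation.
# The cue raises pitch and volume slightly above default.
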SoundCue /BRCosmetics/Sounds/Pets/DwMl/PT_DwMl_HappyStart_Cue
root=QualityLevel_0
nodes [
 {id=QualityLevel_0, children [Random_1,Random_1,WavePlayer_5,WavePlayer_5]}
 {id=Random_1, children [WavePlayer_5,WavePlayer_6,WavePlayer_7,WavePlayer_8], properties {weights [1,1,1,1]}}
 {id=WavePlayer_5, properties {soundWaveAssetPtr=&/PT_DwMl_Treasure_17}}
 {id=WavePlayer_6, properties {soundWaveAssetPtr=&/PT_DwMl_Treasure_17b}}
 {id=WavePlayer_7, properties {soundWaveAssetPtr=&/PT_DwMl_Treasure_17c}}
 {id=WavePlayer_8, properties {soundWaveAssetPtr=&/PT_DwMl_Treasure_17d}}
]
properties {
 attenuationSettings=&Fort_Pets_Close_Athena_Att
 maxDistance=1800.0001
 pitchMultiplier=1.2
 soundClassObject=&Fort_Creatures_Pets
 volumeMultiplier=1.12
}
