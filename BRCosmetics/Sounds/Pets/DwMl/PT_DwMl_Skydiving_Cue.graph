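# Sound cue for the DwMl pet's skydiving sound, routed through the creature pets sound class with close-range pet attenuation.
# A quality switch feeds an equally weighted random pick of two waves on the first two levels, and only the first wave on the other two.
# Volume and pitch are raised, and a concurrency override allows one instance per owner, blocking new plays while one is active.
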SoundCue /BRCosmetics/Sounds/Pets/DwMl/PT_DwMl_Skydiving_Cue
root=QualityLevel_0
nodes [
 {id=QualityLevel_0, children [Random_1,Random_1,WavePlayer_4,WavePlayer_4]}
 {id=Random_1, children [WavePlayer_2,WavePlayer_4], properties {weights [1,1]}}
 {id=WavePlayer_2, properties {soundWaveAssetPtr=&/PT_DwMl_Skydiving_02}}
 {id=WavePlayer_4, properties {soundWaveAssetPtr=&/PT_DwMl_Skydiving_01}}
]
properties {
 attenuationSettings=&Fort_Pets_Close_Athena_Att
 bOverrideConcurrency=true
 maxDistance=1800.0001
 pitchMultiplier=1.2
 soundClassObject=&Fort_Creatures_Pets
 volumeMultiplier=1.344
 concurrencyOverrides {bLimitToOwner=true, maxCount=1, resolutionRule=PreventNew}
}
